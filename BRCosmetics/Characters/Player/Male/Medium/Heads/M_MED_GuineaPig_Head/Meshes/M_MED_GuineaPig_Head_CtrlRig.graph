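# Control Rig for the male guinea pig head cosmetic, procedurally correcting neck hair and beard bones from body and face pose.
# It copies 75% of head swing-twist to head_twist, then sequences weighted parent constraints on neck hair masters.
# Spherical pose readers on head, right upper arm and jaw then weight corrective transforms for hair and beard.
# An unwired head reader and a fur-constraint function sit outside the main execution chain.
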ControlRigBlueprint /BRCosmetics/Characters/Player/Male/Medium/Heads/M_MED_GuineaPig_Head/Meshes/M_MED_GuineaPig_Head_CtrlRig
graphs [
 New Function_ContainedGraph {nodes [
   RigVMFunctionEntryNode()
   > &RigUnit_ParentConstraint(
    AdvancedSettings=(RotationOrderForFilter=XZY)
    Child=(Type=Bone,Name="dyn_head_fur_08")
    Parents=((Item=(Type=Bone,Name="neck_01"),Weight=0.0001),(Item=(Type=Bone,Name="spine_05"),Weight=0.0004),(Item=(Type=Bone,Name="neck_02"),Weight=0.0004),(Item=(Type=Bone,Name="head_twist"),Weight=0.0101),(Item=(Type=Bone,Name="head"),Weight=0.9889))
    Weight=1
   )
   > &RigUnit_ParentConstraint(
    AdvancedSettings=(RotationOrderForFilter=XZY)
    Child=(Type=Bone,Name="dyn_head_fur_10")
    Parents="((Item=(Type=Bone,Name=\"spine_03\"),Weight=0.0002),(Item=(Type=Bone,Name=\"clavicle_l\"),Weight=0.0005),(Item=(Type=Bone,Name=\"clavicle_r\"),Weight=0.0007),(Item=(Type=Bone,Name=\"neck_01\"),Weight=0.0259),(Item=(Type=Bone,Name=\"spine_05\"),Weight=0.0694),(Item=( ...(truncated)"
    Weight=1
   )
   > &RigUnit_ParentConstraint(
    AdvancedSettings=(RotationOrderForFilter=XZY)
    Child=(Type=Bone,Name="dyn_head_fur_11")
    Parents="((Item=(Type=Bone,Name=\"clavicle_l\"),Weight=0.0001),(Item=(Type=Bone,Name=\"neck_01\"),Weight=0.0008),(Item=(Type=Bone,Name=\"spine_05\"),Weight=0.0019),(Item=(Type=Bone,Name=\"neck_02\"),Weight=0.0029),(Item=(Type=Bone,Name=\"head_twist\"),Weight=0.0192),(Item=(T ...(truncated)"
    Weight=1
   )
   > &RigUnit_ParentConstraint(
    AdvancedSettings=(RotationOrderForFilter=XZY)
    Child=(Type=Bone,Name="dyn_head_fur_12")
    Parents=((Item=(Type=Bone,Name="neck_01"),Weight=0.0023),(Item=(Type=Bone,Name="head_twist"),Weight=0.0029),(Item=(Type=Bone,Name="spine_05"),Weight=0.0047),(Item=(Type=Bone,Name="neck_02"),Weight=0.0087),(Item=(Type=Bone,Name="head"),Weight=0.9815))
    Weight=1
   )
   > &RigUnit_ParentConstraint(
    AdvancedSettings=(RotationOrderForFilter=XZY)
    Child=(Type=Bone,Name="dyn_head_fur_13")
    Parents="((Item=(Type=Bone,Name=\"clavicle_l\"),Weight=0.0035),(Item=(Type=Bone,Name=\"clavicle_r\"),Weight=0.0041),(Item=(Type=Bone,Name=\"deform_neck_hair_rear_01\"),Weight=0.0191),(Item=(Type=Bone,Name=\"neck_01\"),Weight=0.063),(Item=(Type=Bone,Name=\"head\"),Weight=0.12 ...(truncated)"
    Weight=1
   )
   > &RigUnit_ParentConstraint(
    AdvancedSettings=(RotationOrderForFilter=XZY)
    Child=(Type=Bone,Name="dyn_head_fur_14")
    Parents="((Item=(Type=Bone,Name=\"deform_neck_hair_d_l_02\"),Weight=0.0067),(Item=(Type=Bone,Name=\"clavicle_l\"),Weight=0.0174),(Item=(Type=Bone,Name=\"head\"),Weight=0.0386),(Item=(Type=Bone,Name=\"deform_neck_hair_rear_01\"),Weight=0.0417),(Item=(Type=Bone,Name=\"neck_01 ...(truncated)"
    Weight=1
   )
   > &RigUnit_ParentConstraint(
    AdvancedSettings=(RotationOrderForFilter=XZY)
    Child=(Type=Bone,Name="dyn_head_fur_15")
    Parents="((Item=(Type=Bone,Name=\"deform_neck_hair_c_l_01\"),Weight=0.0039),(Item=(Type=Bone,Name=\"clavicle_l\"),Weight=0.0277),(Item=(Type=Bone,Name=\"neck_01\"),Weight=0.0578),(Item=(Type=Bone,Name=\"head\"),Weight=0.0805),(Item=(Type=Bone,Name=\"deform_neck_hair_d_l_01\" ...(truncated)"
    Weight=1
   )
   > &RigUnit_ParentConstraint(
    AdvancedSettings=(RotationOrderForFilter=XZY)
    Child=(Type=Bone,Name="dyn_head_fur_16")
    Parents="((Item=(Type=Bone,Name=\"spine_03\"),Weight=0.0003),(Item=(Type=Bone,Name=\"deform_neck_hair_c_l_01\"),Weight=0.0004),(Item=(Type=Bone,Name=\"clavicle_l\"),Weight=0.0035),(Item=(Type=Bone,Name=\"neck_01\"),Weight=0.0107),(Item=(Type=Bone,Name=\"spine_05\"),Weight=0. ...(truncated)"
    Weight=1
   )
   > &RigUnit_ParentConstraint(
    AdvancedSettings=(RotationOrderForFilter=XZY)
    Child=(Type=Bone,Name="dyn_head_fur_17")
    Parents="((Item=(Type=Bone,Name=\"head\"),Weight=0.0025),(Item=(Type=Bone,Name=\"head_twist\"),Weight=0.0045),(Item=(Type=Bone,Name=\"neck_01\"),Weight=0.006),(Item=(Type=Bone,Name=\"neck_02\"),Weight=0.0143),(Item=(Type=Bone,Name=\"clavicle_l\"),Weight=0.029),(Item=(Type=Bo ...(truncated)"
    Weight=1
   )
   > &RigUnit_ParentConstraint(
    AdvancedSettings=(RotationOrderForFilter=XZY)
    Child=(Type=Bone,Name="dyn_head_fur_18")
    Parents="((Item=(Type=Bone,Name=\"upperarm_twist_01_l\"),Weight=0.0004),(Item=(Type=Bone,Name=\"teeth_lower\"),Weight=0.002),(Item=(Type=Bone,Name=\"neck_01\"),Weight=0.004),(Item=(Type=Bone,Name=\"clavicle_l\"),Weight=0.0049),(Item=(Type=Bone,Name=\"neck_02\"),Weight=0.0108 ...(truncated)"
    Weight=1
   )
   > &RigUnit_ParentConstraint(
    AdvancedSettings=(RotationOrderForFilter=XZY)
    Child=(Type=Bone,Name="dyn_head_fur_19")
    Parents="((Item=(Type=Bone,Name=\"neck_01\"),Weight=0.0122),(Item=(Type=Bone,Name=\"neck_02\"),Weight=0.0327),(Item=(Type=Bone,Name=\"clavicle_l\"),Weight=0.0473),(Item=(Type=Bone,Name=\"deform_neck_hair_a_l_01\"),Weight=0.0661),(Item=(Type=Bone,Name=\"spine_05\"),Weight=0.1 ...(truncated)"
    Weight=1
   )
   > &RigUnit_ParentConstraint(
    AdvancedSettings=(RotationOrderForFilter=XZY)
    Child=(Type=Bone,Name="dyn_head_fur_20")
    Parents="((Item=(Type=Bone,Name=\"neck_01\"),Weight=0.0042),(Item=(Type=Bone,Name=\"head_twist\"),Weight=0.0043),(Item=(Type=Bone,Name=\"teeth_lower\"),Weight=0.0048),(Item=(Type=Bone,Name=\"neck_02\"),Weight=0.0119),(Item=(Type=Bone,Name=\"clavicle_l\"),Weight=0.0135),(Item ...(truncated)"
    Weight=1
   )
   > &RigUnit_ParentConstraint(
    AdvancedSettings=(RotationOrderForFilter=XZY)
    Child=(Type=Bone,Name="dyn_head_fur_21")
    Parents="((Item=(Type=Bone,Name=\"C_lip_lower_mid\"),Weight=0.0002),(Item=(Type=Bone,Name=\"R_lip_upper_outer\"),Weight=0.0002),(Item=(Type=Bone,Name=\"L_lip_lower_outer\"),Weight=0.001),(Item=(Type=Bone,Name=\"C_lip_upper_mid\"),Weight=0.0021),(Item=(Type=Bone,Name=\"C_jaw ...(truncated)"
    Weight=1
   )
   > &RigUnit_ParentConstraint(
    AdvancedSettings=(RotationOrderForFilter=XZY)
    Child=(Type=Bone,Name="dyn_head_fur_22")
    Parents="((Item=(Type=Bone,Name=\"neck_02\"),Weight=0.0034),(Item=(Type=Bone,Name=\"def_neck_fur_front\"),Weight=0.0035),(Item=(Type=Bone,Name=\"clavicle_l\"),Weight=0.0056),(Item=(Type=Bone,Name=\"L_lip_corner\"),Weight=0.0125),(Item=(Type=Bone,Name=\"spine_05\"),Weight=0.0 ...(truncated)"
    Weight=1
   )
   > &RigUnit_ParentConstraint(
    AdvancedSettings=(RotationOrderForFilter=XZY)
    Child=(Type=Bone,Name="dyn_head_fur_23")
    Parents="((Item=(Type=Bone,Name=\"spine_03\"),Weight=0.0001),(Item=(Type=Bone,Name=\"deform_neck_hair_d_r_01\"),Weight=0.0002),(Item=(Type=Bone,Name=\"clavicle_r\"),Weight=0.0045),(Item=(Type=Bone,Name=\"neck_01\"),Weight=0.0611),(Item=(Type=Bone,Name=\"neck_02\"),Weight=0.1 ...(truncated)"
    Weight=1
   )
   > &RigUnit_ParentConstraint(
    AdvancedSettings=(RotationOrderForFilter=XZY)
    Child=(Type=Bone,Name="dyn_head_fur_29")
    Parents="((Item=(Type=Bone,Name=\"clavicle_r\"),Weight=0.0006),(Item=(Type=Bone,Name=\"neck_01\"),Weight=0.004),(Item=(Type=Bone,Name=\"spine_05\"),Weight=0.0098),(Item=(Type=Bone,Name=\"neck_02\"),Weight=0.0157),(Item=(Type=Bone,Name=\"head_twist\"),Weight=0.0192),(Item=(Ty ...(truncated)"
    Weight=1
   )
   > &RigUnit_ParentConstraint(
    AdvancedSettings=(RotationOrderForFilter=XZY)
    Child=(Type=Bone,Name="dyn_head_fur_31")
    Parents="((Item=(Type=Bone,Name=\"upperarm_twist_01_r\"),Weight=0.0006),(Item=(Type=Bone,Name=\"deform_upperarm_up_r\"),Weight=0.001),(Item=(Type=Bone,Name=\"neck_01\"),Weight=0.0184),(Item=(Type=Bone,Name=\"clavicle_r\"),Weight=0.0402),(Item=(Type=Bone,Name=\"neck_02\"),Wei ...(truncated)"
    Weight=1
   )
   > &RigUnit_ParentConstraint(
    AdvancedSettings=(RotationOrderForFilter=XZY)
    Child=(Type=Bone,Name="dyn_head_fur_32")
    Parents=((Item=(Type=Bone,Name="head_twist"),Weight=0.0001),(Item=(Type=Bone,Name="neck_02"),Weight=0.0002),(Item=(Type=Bone,Name="spine_05"),Weight=0.0002),(Item=(Type=Bone,Name="head"),Weight=0.9995))
    Weight=1
   )
   > &RigUnit_ParentConstraint(
    AdvancedSettings=(RotationOrderForFilter=XZY)
    Child=(Type=Bone,Name="dyn_head_fur_33")
    Parents="((Item=(Type=Bone,Name=\"upperarm_twist_01_r\"),Weight=0.0018),(Item=(Type=Bone,Name=\"deform_upperarm_up_r\"),Weight=0.0027),(Item=(Type=Bone,Name=\"neck_01\"),Weight=0.01),(Item=(Type=Bone,Name=\"clavicle_r\"),Weight=0.0259),(Item=(Type=Bone,Name=\"neck_02\"),Weig ...(truncated)"
    Weight=1
   )
   > &RigUnit_ParentConstraint(
    AdvancedSettings=(RotationOrderForFilter=XZY)
    Child=(Type=Bone,Name="dyn_head_fur_34")
    Parents="((Item=(Type=Bone,Name=\"deform_upperarm_up_r\"),Weight=0.0081),(Item=(Type=Bone,Name=\"neck_01\"),Weight=0.0116),(Item=(Type=Bone,Name=\"deform_neck_hair_b_r_01\"),Weight=0.0303),(Item=(Type=Bone,Name=\"neck_02\"),Weight=0.0359),(Item=(Type=Bone,Name=\"clavicle_r\" ...(truncated)"
    Weight=1
   )
   > &RigUnit_ParentConstraint(
    AdvancedSettings=(RotationOrderForFilter=XZY)
    Child=(Type=Bone,Name="dyn_head_fur_35")
    Parents=((Item=(Type=Bone,Name="neck_02"),Weight=0.0002),(Item=(Type=Bone,Name="spine_05"),Weight=0.0002),(Item=(Type=Bone,Name="R_cheek_inner"),Weight=0.0211),(Item=(Type=Bone,Name="head"),Weight=0.9785))
    Weight=1
   )
   > &RigUnit_ParentConstraint(
    AdvancedSettings=(RotationOrderForFilter=XZY)
    Child=(Type=Bone,Name="dyn_head_fur_37")
    Parents=((Item=(Type=Bone,Name="R_brow_mid"),Weight=0.0006),(Item=(Type=Bone,Name="R_brow_outer"),Weight=0.2879),(Item=(Type=Bone,Name="head"),Weight=0.7115))
    Weight=1
   )
   > &RigUnit_ParentConstraint(
    AdvancedSettings=(RotationOrderForFilter=XZY)
    Child=(Type=Bone,Name="dyn_head_fur_38")
    Parents="((Item=(Type=Bone,Name=\"teeth_lower\"),Weight=0.0009),(Item=(Type=Bone,Name=\"clavicle_r\"),Weight=0.0018),(Item=(Type=Bone,Name=\"neck_01\"),Weight=0.0023),(Item=(Type=Bone,Name=\"C_jaw\"),Weight=0.0038),(Item=(Type=Bone,Name=\"neck_02\"),Weight=0.0074),(Item=(Typ ...(truncated)"
    Weight=1
   )
   > &RigUnit_ParentConstraint(
    AdvancedSettings=(RotationOrderForFilter=XZY)
    Child=(Type=Bone,Name="dyn_head_fur_39")
    Parents="((Item=(Type=Bone,Name=\"L_lip_upper_outer\"),Weight=0.0002),(Item=(Type=Bone,Name=\"R_lip_lower_outer\"),Weight=0.0006),(Item=(Type=Bone,Name=\"C_lip_upper_mid\"),Weight=0.002),(Item=(Type=Bone,Name=\"C_jaw\"),Weight=0.0041),(Item=(Type=Bone,Name=\"R_lip_upper_out ...(truncated)"
    Weight=1
   )
   > &RigUnit_ParentConstraint(
    AdvancedSettings=(RotationOrderForFilter=XZY)
    Child=(Type=Bone,Name="dyn_head_fur_40")
    Parents="((Item=(Type=Bone,Name=\"neck_02\"),Weight=0.0036),(Item=(Type=Bone,Name=\"deform_neck_hair_a_r_01\"),Weight=0.006),(Item=(Type=Bone,Name=\"spine_05\"),Weight=0.0119),(Item=(Type=Bone,Name=\"R_lip_lower_outer\"),Weight=0.0163),(Item=(Type=Bone,Name=\"R_lip_upper_ou ...(truncated)"
    Weight=1
   )
   > &RigUnit_ParentConstraint(
    AdvancedSettings=(RotationOrderForFilter=XZY)
    Child=(Type=Bone,Name="dyn_head_fur_41")
    Parents="((Item=(Type=Bone,Name=\"L_lip_corner\"),Weight=0.002),(Item=(Type=Bone,Name=\"L_lip_upper_outer\"),Weight=0.0359),(Item=(Type=Bone,Name=\"R_lip_corner\"),Weight=0.037),(Item=(Type=Bone,Name=\"R_lip_upper_outer\"),Weight=0.1445),(Item=(Type=Bone,Name=\"C_lip_upper_ ...(truncated)"
    Weight=1
   )
   > &RigUnit_ParentConstraint(
    AdvancedSettings=(RotationOrderForFilter=XZY)
    Child=(Type=Bone,Name="dyn_head_fur_42")
    Parents="((Item=(Type=Bone,Name=\"L_lip_lower_outer\"),Weight=0.0001),(Item=(Type=Bone,Name=\"R_lip_corner\"),Weight=0.0005),(Item=(Type=Bone,Name=\"R_lip_upper_outer\"),Weight=0.0102),(Item=(Type=Bone,Name=\"C_lip_upper_mid\"),Weight=0.0635),(Item=(Type=Bone,Name=\"L_lip_u ...(truncated)"
    Weight=1
   )
   > &RigUnit_ParentConstraint(
    AdvancedSettings=(RotationOrderForFilter=XZY)
    Child=(Type=Bone,Name="dyn_head_fur_46")
    Parents=((Item=(Type=Bone,Name="R_brow_outer"),Weight=0.0097),(Item=(Type=Bone,Name="R_brow_mid"),Weight=0.0314),(Item=(Type=Bone,Name="head"),Weight=0.9589))
    Weight=1
   )
   > &RigUnit_ParentConstraint(
    AdvancedSettings=(RotationOrderForFilter=XZY)
    Child=(Type=Bone,Name="dyn_head_fur_48")
    Parents=((Item=(Type=Bone,Name="L_brow_mid"),Weight=0.0003),(Item=(Type=Bone,Name="C_brow_mid"),Weight=0.0008),(Item=(Type=Bone,Name="R_brow_outer"),Weight=0.0921),(Item=(Type=Bone,Name="R_brow_mid"),Weight=0.2897),(Item=(Type=Bone,Name="head"),Weight=0.6171))
    Weight=1
   )
   > &RigUnit_ParentConstraint(
    AdvancedSettings=(RotationOrderForFilter=XZY)
    Child=(Type=Bone,Name="dyn_head_fur_49")
    Parents=((Item=(Type=Bone,Name="R_brow_mid"),Weight=0.0003),(Item=(Type=Bone,Name="C_brow_mid"),Weight=0.0008),(Item=(Type=Bone,Name="L_brow_outer"),Weight=0.0921),(Item=(Type=Bone,Name="L_brow_mid"),Weight=0.2897),(Item=(Type=Bone,Name="head"),Weight=0.6171))
    Weight=1
   )
   > &RigUnit_ParentConstraint(
    AdvancedSettings=(InterpolationType=Average,RotationOrderForFilter=XZY)
    Child=(Type=Bone,Name="dyn_head_fur_50")
    Parents=((Item=(Type=Bone,Name="deform_neck_hair_rear_02"),Weight=1))
    Weight=1
   )
   > RigVMFunctionReturnNode()
  ]}
 RigVMFunctionLibrary {nodes [Pin = RigVMCollapseNode(ExecuteContext=())]}
 RigVMFunction_Sequence_1_ContainedGraph {nodes [
   n1 = RigVMFunctionEntryNode()
   RigVMFunction_Sequence = &RigVMFunction_Sequence(ExecuteContext=@n1.ExecuteContext)
   n2 = &RigVMFunction_Sequence(ExecuteContext=@RigVMFunction_Sequence.B)
   n3 = &RigVMFunction_Sequence(ExecuteContext=@n2.B)
   RigVMFunctionReturnNode(A=@RigVMFunction_Sequence.A, B=@n2.A, C=@n3.A, D=@n3.B)
  ]}
 RigVMModel {nodes [
   RigUnit_BeginExecution = &RigUnit_BeginExecution()
   RigUnit_GetTransform = &RigUnit_GetTransform(Item=(Type=Bone,Name="head"), Space=LocalSpace, bInitial=False)
   RigVMFunction_MathQuaternionSwingTwist = &RigVMFunction_MathQuaternionSwingTwist(TwistAxis=(X=1,Y=0,Z=0), Input=@RigUnit_GetTransform.Transform.Rotation)
   RigUnit_SetRotation = &RigUnit_SetRotation(
    Item=(Type=Bone,Name="head_twist")
    Weight=0.75
    bInitial=False
    ExecutePin=@RigUnit_BeginExecution.ExecutePin
    Value=@RigVMFunction_MathQuaternionSwingTwist.Twist
   )
   > RigVMFunction_Sequence = RigVMAggregateNode()
   &RigUnit_ParentConstraint(
    AdvancedSettings=(RotationOrderForFilter=XZY)
    Child=(Type=Bone,Name="deform_neck_hair_a_l_MSTR")
    Parents="((Item=(Type=Bone,Name=\"def_neck_fur_front\"),Weight=0.0044),(Item=(Type=Bone,Name=\"clavicle_l\"),Weight=0.0144),(Item=(Type=Bone,Name=\"teeth_lower\"),Weight=0.0175),(Item=(Type=Bone,Name=\"neck_01\"),Weight=0.0177),(Item=(Type=Bone,Name=\"neck_02\"),Weight=0.049 ...(truncated)"
    Weight=1
    ExecutePin=@RigVMFunction_Sequence.A
   )
   > &RigUnit_ParentConstraint(
    AdvancedSettings=(RotationOrderForFilter=XZY)
    Child=(Type=Bone,Name="deform_neck_hair_b_l_MSTR")
    Parents="((Item=(Type=Bone,Name=\"teeth_lower\"),Weight=0.0041),(Item=(Type=Bone,Name=\"spine_04\"),Weight=0.0049),(Item=(Type=Bone,Name=\"upperarm_twist_01_l\"),Weight=0.0062),(Item=(Type=Bone,Name=\"neck_01\"),Weight=0.0423),(Item=(Type=Bone,Name=\"clavicle_l\"),Weight=0.0 ...(truncated)"
    Weight=1
   )
   > &RigUnit_ParentConstraint(
    AdvancedSettings=(RotationOrderForFilter=XZY)
    Child=(Type=Bone,Name="deform_neck_hair_c_l_MSTR")
    Parents="((Item=(Type=Bone,Name=\"spine_03\"),Weight=0.0011),(Item=(Type=Bone,Name=\"spine_04\"),Weight=0.0033),(Item=(Type=Bone,Name=\"upperarm_twist_01_l\"),Weight=0.0058),(Item=(Type=Bone,Name=\"neck_01\"),Weight=0.0319),(Item=(Type=Bone,Name=\"head\"),Weight=0.0329),(Ite ...(truncated)"
    Weight=1
   )
   > &RigUnit_ParentConstraint(
    AdvancedSettings=(RotationOrderForFilter=XZY)
    Child=(Type=Bone,Name="deform_neck_hair_d_l_MSTR")
    Parents="((Item=(Type=Bone,Name=\"spine_03\"),Weight=0.0011),(Item=(Type=Bone,Name=\"head\"),Weight=0.0231),(Item=(Type=Bone,Name=\"clavicle_l\"),Weight=0.0722),(Item=(Type=Bone,Name=\"neck_01\"),Weight=0.0743),(Item=(Type=Bone,Name=\"neck_02\"),Weight=0.1696),(Item=(Type=Bo ...(truncated)"
    Weight=1
   )
   > &RigUnit_ParentConstraint(
    AdvancedSettings=(RotationOrderForFilter=XZY)
    Child=(Type=Bone,Name="deform_neck_hair_rear_MSTR")
    Parents="((Item=(Type=Bone,Name=\"spine_03\"),Weight=0.0004),(Item=(Type=Bone,Name=\"head\"),Weight=0.0179),(Item=(Type=Bone,Name=\"clavicle_r\"),Weight=0.0421),(Item=(Type=Bone,Name=\"neck_01\"),Weight=0.0568),(Item=(Type=Bone,Name=\"neck_02\"),Weight=0.127),(Item=(Type=Bon ...(truncated)"
    Weight=1
   )
   > &RigUnit_ParentConstraint(
    AdvancedSettings=(RotationOrderForFilter=XZY)
    Child=(Type=Bone,Name="deform_neck_hair_a_r_MSTR")
    Parents="((Item=(Type=Bone,Name=\"clavicle_r\"),Weight=0.0001),(Item=(Type=Bone,Name=\"def_neck_fur_front\"),Weight=0.0001),(Item=(Type=Bone,Name=\"teeth_lower\"),Weight=0.0027),(Item=(Type=Bone,Name=\"neck_01\"),Weight=0.0144),(Item=(Type=Bone,Name=\"neck_02\"),Weight=0.052 ...(truncated)"
    Weight=1
   )
   > &RigUnit_ParentConstraint(
    AdvancedSettings=(RotationOrderForFilter=XZY)
    Child=(Type=Bone,Name="deform_neck_hair_b_r_MSTR")
    Parents="((Item=(Type=Bone,Name=\"spine_03\"),Weight=0.0005),(Item=(Type=Bone,Name=\"spine_04\"),Weight=0.0017),(Item=(Type=Bone,Name=\"teeth_lower\"),Weight=0.0026),(Item=(Type=Bone,Name=\"neck_01\"),Weight=0.0681),(Item=(Type=Bone,Name=\"clavicle_r\"),Weight=0.0721),(Item= ...(truncated)"
    Weight=1
   )
   > &RigUnit_ParentConstraint(
    AdvancedSettings=(RotationOrderForFilter=XZY)
    Child=(Type=Bone,Name="deform_neck_hair_c_r_MSTR")
    Parents="((Item=(Type=Bone,Name=\"spine_03\"),Weight=0.0006),(Item=(Type=Bone,Name=\"spine_04\"),Weight=0.0011),(Item=(Type=Bone,Name=\"head\"),Weight=0.0148),(Item=(Type=Bone,Name=\"neck_01\"),Weight=0.028),(Item=(Type=Bone,Name=\"neck_02\"),Weight=0.0702),(Item=(Type=Bone, ...(truncated)"
    Weight=1
   )
   > &RigUnit_ParentConstraint(
    AdvancedSettings=(RotationOrderForFilter=XZY)
    Child=(Type=Bone,Name="deform_neck_hair_d_r_MSTR")
    Parents="((Item=(Type=Bone,Name=\"spine_03\"),Weight=0.0008),(Item=(Type=Bone,Name=\"head\"),Weight=0.023),(Item=(Type=Bone,Name=\"clavicle_r\"),Weight=0.0723),(Item=(Type=Bone,Name=\"neck_01\"),Weight=0.0742),(Item=(Type=Bone,Name=\"neck_02\"),Weight=0.1696),(Item=(Type=Bon ...(truncated)"
    Weight=1
   )
   n1 = &RigUnit_SphericalPoseReader(
    ActiveRegionScaleFactors=(PositiveWidth=3.5,NegativeWidth=3.5,PositiveHeight=1,NegativeHeight=3.5)
    ActiveRegionSize=0.1
    Debug=(bDrawDebug=false,bDraw2D=False,bDrawLocalAxes=False,DebugScale=25,DebugSegments=20,DebugThickness=0.25)
    DriverAxis=(X=0,Y=1,Z=0)
    DriverItem=(Type=Bone,Name="head")
    FalloffRegionScaleFactors=(PositiveWidth=1,NegativeWidth=1,PositiveHeight=1,NegativeHeight=1)
    FalloffSize=0.35
    FlipHeightScaling=False
    FlipWidthScaling=False
    OptionalParentItem=(Type=Bone,Name="spine_05")
    RotationOffset=(X=20,Y=180,Z=0)
    ExecutePin=@RigVMFunction_Sequence.B
   )
   > &RigUnit_ModifyTransforms(
    ItemToModify="((Item=(Type=Bone,Name=\"deform_neck_hair_a_l_MSTR\"),Transform=(Rotation=(X=0,Y=0,Z=0,W=1),Translation=(X=0,Y=-2,Z=0),Scale3D=(X=1,Y=1,Z=1))),(Item=(Type=Bone,Name=\"deform_neck_hair_b_l_MSTR\"),Transform=(Rotation=(X=0,Y=0,Z=0,W=1),Translation=(X=0,Y=-2,Z=0) ...(truncated)"
    WeightMinimum=0
    Weight=@n1.OutputParam
   )
   > n2 = &RigUnit_SphericalPoseReader(
    ActiveRegionScaleFactors=(PositiveWidth=1,NegativeWidth=1,PositiveHeight=1,NegativeHeight=1)
    ActiveRegionSize=0.1
    Debug=(bDrawDebug=true,bDraw2D=False,bDrawLocalAxes=False,DebugScale=25,DebugSegments=20,DebugThickness=0.25)
    DriverAxis=(X=1,Y=0,Z=0)
    DriverItem=(Type=Bone,Name="head")
    FalloffRegionScaleFactors=(PositiveWidth=1,NegativeWidth=1,PositiveHeight=1,NegativeHeight=1)
    FalloffSize=0.35
    FlipHeightScaling=False
    FlipWidthScaling=False
    OptionalParentItem=(Type=Bone,Name="spine_05")
    RotationOffset=(X=-90,Y=0,Z=20)
   )
   > &RigUnit_ModifyTransforms(
    ItemToModify="((Item=(Type=Bone,Name=\"deform_neck_hair_d_l_MSTR\"),Transform=(Rotation=(X=0.003,Y=-0.1115,Z=0.0146,W=0.9936),Translation=(X=0,Y=0,Z=0),Scale3D=(X=1,Y=1,Z=1))),(Item=(Type=Bone,Name=\"deform_neck_hair_b_r_MSTR\"),Transform=(Rotation=(X=0,Y=0,Z=0,W=1),Transla ...(truncated)"
    WeightMinimum=0
    Weight=@n2.OutputParam
   )
   > n3 = &RigUnit_SphericalPoseReader(
    ActiveRegionScaleFactors=(PositiveWidth=0.5,NegativeWidth=3,PositiveHeight=1.5,NegativeHeight=1.5)
    ActiveRegionSize=0.1
    Debug=(bDrawDebug=false,DebugScale=25,DebugSegments=20,DebugThickness=0.25)
    DriverAxis=(X=-1,Y=0,Z=0)
    DriverItem=(Type=Bone,Name="upperarm_r")
    FalloffRegionScaleFactors=(PositiveWidth=1,NegativeWidth=1,PositiveHeight=1,NegativeHeight=1)
    FalloffSize=0.4
    FlipWidthScaling=true
    OptionalParentItem=(Type=Bone)
    RotationOffset=(X=90,Y=180,Z=180)
   )
   > &RigUnit_ModifyTransforms(
    ItemToModify="((Item=(Type=Bone,Name=\"deform_neck_hair_b_r_MSTR\"),Transform=(Rotation=(X=-0.1404,Y=0.0426,Z=0.2871,W=0.9466),Translation=(X=0,Y=0,Z=0),Scale3D=(X=1,Y=1,Z=1))),(Item=(Type=Bone,Name=\"deform_neck_hair_c_r_MSTR\"),Transform=(Rotation=(X=-0.0077,Y=0.0233,Z=0. ...(truncated)"
    WeightMinimum=0
    Weight=@n3.OutputParam
   )
   n4 = &RigUnit_SphericalPoseReader(
    ActiveRegionScaleFactors=(PositiveWidth=1,NegativeWidth=1,PositiveHeight=1,NegativeHeight=1)
    ActiveRegionSize=0.1
    Debug=(bDrawDebug=false,bDraw2D=False,bDrawLocalAxes=False,DebugScale=25,DebugSegments=20,DebugThickness=0.25)
    DriverAxis=(X=0,Y=1,Z=0)
    DriverItem=(Type=Bone,Name="C_jaw")
    FalloffRegionScaleFactors=(PositiveWidth=1,NegativeWidth=1,PositiveHeight=1,NegativeHeight=1)
    FalloffSize=0.26
    FlipHeightScaling=False
    FlipWidthScaling=False
    OptionalParentItem=(Type=Bone,Name="None")
    RotationOffset=(X=138,Y=0,Z=0)
    ExecutePin=@RigVMFunction_Sequence.C
   )
   > &RigUnit_ModifyTransforms(
    ItemToModify=((Item=(Type=Bone,Name="dyn_beard_MSTR"),Transform=(Rotation=(X=0.4617,Y=0,Z=0,W=0.887),Translation=(X=0,Y=1,Z=0.5),Scale3D=(X=1,Y=1,Z=1))))
    WeightMinimum=0
    Weight=@n4.OutputParam
   )
   n5 = &RigUnit_SphericalPoseReader(
    ActiveRegionScaleFactors=(PositiveWidth=3.5,NegativeWidth=3.5,PositiveHeight=1,NegativeHeight=3.5)
    ActiveRegionSize=0.1
    Debug=(bDrawDebug=false,bDraw2D=False,bDrawLocalAxes=False,DebugScale=25,DebugSegments=20,DebugThickness=0.25)
    DriverAxis=(X=0,Y=1,Z=0)
    DriverItem=(Type=Bone,Name="head")
    ExecutePin=()
    FalloffRegionScaleFactors=(PositiveWidth=1,NegativeWidth=1,PositiveHeight=1,NegativeHeight=1)
    FalloffSize=0.35
    FlipHeightScaling=False
    FlipWidthScaling=False
    OptionalParentItem=(Type=Bone,Name="None")
    RotationOffset=(X=20,Y=0,Z=0)
   )
   > &RigUnit_ModifyTransforms(
    ItemToModify="((Item=(Type=Bone,Name=\"deform_neck_hair_a_l_MSTR\"),Transform=(Rotation=(X=0,Y=0,Z=0,W=1),Translation=(X=0,Y=0,Z=0),Scale3D=(X=1,Y=1,Z=1))),(Item=(Type=Bone,Name=\"deform_neck_hair_b_l_MSTR\"),Transform=(Rotation=(X=0,Y=0,Z=0,W=1),Translation=(X=0,Y=0,Z=0),S ...(truncated)"
    WeightMinimum=0
    Weight=@n5.OutputParam
   )
  ]}
]
hierarchy [
 root {loc=<0 0 0>, rot=<0 0 0 1>}
 pelvis {parent=root, loc=<0 -0.7071 93.1191>, rot=<0.0127 0.707 -0.0127 -0.707>}
 spine_01 {parent=pelvis, loc=<6.7494 -0.6388 0>, rot=<0 0 0.034 -0.9994>}
 spine_02 {parent=spine_01, loc=<7.2518 -0.035 0>, rot=<0 0 0.0238 -0.9997>}
 spine_03 {parent=spine_02, loc=<6.0104 0 0>, rot=<0 0 0.0149 -0.9999>}
 spine_04 {parent=spine_03, loc=<8.1024 0 0>, rot=<0 0 -0.0132 -0.9999>}
 spine_05 {parent=spine_04, loc=<10.0146 0 0>, rot=<0 0 -0.0015 -1>}
 neck_01 {parent=spine_05, loc=<11.3138 0.1529 0>, rot=<0 0 -0.2003 -0.9797>}
 neck_02 {parent=neck_01, loc=<4.5627 0 0>, rot=<0 0 0.0504 -0.9987>}
 head {parent=neck_02, loc=<5.3914 0 0>, rot=<0 0 0.115 -0.9934>}
 faceAttach {parent=head, loc=<0 0 0>, rot=<0.0029 -0.7071 -0.0029 -0.7071>}
 C_brow_mid {parent=faceAttach, loc=<0 13.0448 7.6773>, rot=<0 0 0 -1>}
 C_jaw {parent=faceAttach, loc=<-0.0134 8.0146 0.3>, rot=<0 0 0 1>}
 C_lip_lower_mid {parent=C_jaw, loc=<0.0223 6.9593 -1.4987>, rot=<0 0 0 -1>}
 L_lip_lower_outer {parent=C_jaw, loc=<1.7592 6.3307 -1.4414>, rot=<-0.004 0.0068 0.1992 -0.9799>}
 R_lip_lower_outer {parent=C_jaw, loc=<-1.7324 6.3307 -1.4414>, rot=<-0.004 -0.0068 -0.1992 -0.9799>}
 teeth_lower {parent=C_jaw, loc=<0.0134 2.3689 -1.8377>, rot=<0 0 0 1>}
 tongue {parent=C_jaw, loc=<0.0141 2.7612 -0.1302>, rot=<0.0518 0 0 -0.9987>}
 dyn_beard_MSTR {parent=C_jaw, loc=<0 5.962 -4.4682>, rot=<0 0 0 1>}
 dyn_beard_l_MSTR {parent=dyn_beard_MSTR, loc=<2.4043 -1.4568 -2.8033>, rot=<-0.5012 -0.3467 0.7319 -0.3048>}
 dyn_beard_l_01 {parent=dyn_beard_l_MSTR, loc=<0 0 0>, rot=<0 0 0 1>}
 dyn_beard_l_02 {parent=dyn_beard_l_01, loc=<1.1882 0 0>, rot=<0.7801 -0.0032 -0.0109 -0.6256>}
 dyn_beard_c_MSTR {parent=dyn_beard_MSTR, loc=<0.3766 -0.3039 -2.6214>, rot=<-0.4058 -0.5341 0.6014 -0.4341>}
 dyn_beard_c_01 {parent=dyn_beard_c_MSTR, loc=<0 0 0>, rot=<0 0 0 1>}
 dyn_beard_c_02 {parent=dyn_beard_c_01, loc=<1.3962 0 0>, rot=<-0.0329 0.0288 -0.0856 -0.9954>}
 dyn_beard_r_MSTR {parent=dyn_beard_MSTR, loc=<-2.101 -1.2607 -2.1819>, rot=<-0.352 -0.5178 0.4071 -0.665>}
 dyn_beard_r_01 {parent=dyn_beard_r_MSTR, loc=<0 0 0>, rot=<0 0 0 1>}
 dyn_beard_r_02 {parent=dyn_beard_r_01, loc=<1.9995 0 0>, rot=<0.5979 0.1645 0.0432 -0.7833>}
 C_lip_upper_mid {parent=faceAttach, loc=<-0.0233 14.8373 0.056>, rot=<-0.1219 0.0011 -0.0009 -0.9925>}
 L_brow_mid {parent=faceAttach, loc=<2.6832 12.2852 8.1254>, rot=<-0.1117 0.0088 0.116 -0.9869>}
 L_brow_outer {parent=faceAttach, loc=<5.2456 11.2606 8.0909>, rot=<-0.0515 -0.0263 0.2236 -0.973>}
 L_cheek_inner {parent=faceAttach, loc=<4.8472 11.2445 2.5996>, rot=<-0.2037 -0.1078 0.2662 -0.936>}
 L_eye {parent=faceAttach, loc=<3.106 8.1912 5.9129>, rot=<0 0 0.9998 0.0175>}
 L_eye_lid_lower_mid {parent=L_eye, loc=<0 0 0>, rot=<-0.0707 0 0 -0.9975>}
 L_eye_lid_upper_mid {parent=L_eye, loc=<0 0 0>, rot=<0.0727 0 0 -0.9974>}
 L_lip_corner {parent=faceAttach, loc=<3.2348 13.0542 -0.7297>, rot=<0.003 0.008 -0.0146 -0.9999>}
 L_lip_upper_outer {parent=faceAttach, loc=<1.7557 14.3449 -0.0633>, rot=<-0.0732 -0.0115 0.2302 -0.9703>}
 R_brow_mid {parent=faceAttach, loc=<-2.6832 12.2852 8.1254>, rot=<-0.1117 -0.0088 -0.116 -0.9869>}
 R_brow_outer {parent=faceAttach, loc=<-5.2456 11.2606 8.0909>, rot=<-0.0515 0.0263 -0.2236 -0.973>}
 R_cheek_inner {parent=faceAttach, loc=<-4.8472 11.2445 2.5996>, rot=<-0.2037 0.1078 -0.2662 -0.936>}
 R_eye {parent=faceAttach, loc=<-3.106 8.1912 5.9128>, rot=<0 0 0.9998 -0.0175>}
 R_eye_lid_lower_mid {parent=R_eye, loc=<0 0 0.0001>, rot=<-0.0707 0 0 -0.9975>}
 R_eye_lid_upper_mid {parent=R_eye, loc=<0 0 0.0001>, rot=<0.0727 0 0 -0.9974>}
 R_lip_corner {parent=faceAttach, loc=<-3.2348 13.0542 -0.7297>, rot=<0.003 -0.008 0.0146 -0.9999>}
 R_lip_upper_outer {parent=faceAttach, loc=<-1.7557 14.3449 -0.0633>, rot=<-0.0732 0.0115 -0.2302 -0.9703>}
 teeth_upper {parent=faceAttach, loc=<0 9.7287 0.7952>, rot=<0 0 0 1>}
 C_nose_bridge {parent=faceAttach, loc=<0 13.5309 4.6814>, rot=<-0.4193 0 0 -0.9079>}
 deform_neck_hair_a_l_MSTR {parent=head, loc=<-4.0087 5.6409 -6.2811>, rot=<-0.0067 -0.9661 -0.1741 -0.1906>}
 deform_neck_hair_a_l_01 {parent=deform_neck_hair_a_l_MSTR, loc=<0 0 0>, rot=<0 0 0 1>}
 deform_neck_hair_a_l_02 {parent=deform_neck_hair_a_l_01, loc=<3.9204 0 0>, rot=<0 0 0 1>}
 deform_neck_hair_b_l_MSTR {parent=head, loc=<-2.4611 2.1755 -8.4027>, rot=<0.1928 -0.9565 -0.0696 -0.2074>}
 deform_neck_hair_b_l_01 {parent=deform_neck_hair_b_l_MSTR, loc=<0 0 0>, rot=<0 0 0 1>}
 deform_neck_hair_b_l_02 {parent=deform_neck_hair_b_l_01, loc=<5.973 0 0>, rot=<0 0.2462 0 -0.9692>}
 deform_neck_hair_c_l_MSTR {parent=head, loc=<-5.6305 -5.3382 -10.1057>, rot=<0.2354 -0.9056 0.2318 -0.2658>}
 deform_neck_hair_c_l_01 {parent=deform_neck_hair_c_l_MSTR, loc=<0 0 0>, rot=<0 0 0 1>}
 deform_neck_hair_c_l_02 {parent=deform_neck_hair_c_l_01, loc=<2.309 0 0>, rot=<0 0 0 1>}
 deform_neck_hair_d_l_MSTR {parent=head, loc=<-4.8587 -9.2115 -5.2715>, rot=<0.229 -0.681 0.6596 -0.2209>}
 deform_neck_hair_d_l_01 {parent=deform_neck_hair_d_l_MSTR, loc=<0 0 0>, rot=<0 0 0 1>}
 deform_neck_hair_d_l_02 {parent=deform_neck_hair_d_l_01, loc=<3.9434 0 0>, rot=<0 0 0 1>}
 deform_neck_hair_rear_MSTR {parent=head, loc=<-5.2893 -10.0691 -0.1196>, rot=<0.229 -0.681 0.6596 -0.2209>}
 deform_neck_hair_rear_01 {parent=deform_neck_hair_rear_MSTR, loc=<0 0 0>, rot=<0 -0.0343 0 -0.9994>}
 deform_neck_hair_rear_02 {parent=deform_neck_hair_rear_01, loc=<5.0389 0 0>, rot=<0 0 0 1>}
 deform_neck_hair_a_r_MSTR {parent=head, loc=<-4.009 5.6409 6.2811>, rot=<0.9661 -0.0067 -0.1906 0.1741>}
 deform_neck_hair_a_r_01 {parent=deform_neck_hair_a_r_MSTR, loc=<0 0 0>, rot=<0 0 0 -1>}
 deform_neck_hair_a_r_02 {parent=deform_neck_hair_a_r_01, loc=<-3.92 0 -0.0002>, rot=<0 0 0 -1>}
 deform_neck_hair_b_r_MSTR {parent=head, loc=<-2.4613 2.1755 8.4027>, rot=<0.9565 0.1928 -0.2074 0.0696>}
 deform_neck_hair_b_r_01 {parent=deform_neck_hair_b_r_MSTR, loc=<0 0 0>, rot=<0 0 0 1>}
 deform_neck_hair_b_r_02 {parent=deform_neck_hair_b_r_01, loc=<-5.9728 0.0001 -0.0001>, rot=<0 0.2462 0 -0.9692>}
 deform_neck_hair_c_r_MSTR {parent=head, loc=<-5.6303 -5.3382 10.1057>, rot=<0.9056 0.2354 -0.2658 -0.2318>}
 deform_neck_hair_c_r_01 {parent=deform_neck_hair_c_r_MSTR, loc=<0 0 0>, rot=<0 0 0 1>}
 deform_neck_hair_c_r_02 {parent=deform_neck_hair_c_r_01, loc=<-2.3088 0.0001 -0.0001>, rot=<0 0 0 1>}
 deform_neck_hair_d_r_MSTR {parent=head, loc=<-4.8583 -9.2115 5.2715>, rot=<0.681 0.229 -0.2209 -0.6596>}
 deform_neck_hair_d_r_01 {parent=deform_neck_hair_d_r_MSTR, loc=<0 0 0>, rot=<0 0 0 1>}
 deform_neck_hair_d_r_02 {parent=deform_neck_hair_d_r_01, loc=<-3.944 0 0.0004>, rot=<0 0 0 1>}
 head_twist {parent=neck_01, loc=<4.5627 0 0>, rot=<0 0 0.0504 -0.9987>}
 clavicle_l {parent=spine_05, loc=<4.3575 -0.1466 -2.3223>, rot=<0.0465 -0.7289 -0.009 -0.6829>}
 upperarm_l {parent=clavicle_l, loc=<15.5111 0 0>, rot=<0.0226 -0.3234 -0.0088 -0.9459>}
 upperarm_twist_01_l {parent=upperarm_l, loc=<0.0041 0 0>, rot=<0 0 0 1>}
 deform_upperarm_up_l {parent=upperarm_twist_01_l, loc=<-3.2216 -1.1391 3.3311>, rot=<0 0 0 1>}
 clavicle_r {parent=spine_05, loc=<4.3575 -0.1466 2.3223>, rot=<0.7289 0.0465 -0.6829 0.009>}
 upperarm_r {parent=clavicle_r, loc=<-15.5111 0 0>, rot=<0.0226 -0.3234 -0.0088 -0.9459>}
 upperarm_twist_01_r {parent=upperarm_r, loc=<-0.0041 0 0>, rot=<0 0 0 -1>}
 deform_upperarm_fr_r {parent=upperarm_twist_01_r, loc=<0.3689 -7.0506 2.7652>, rot=<0 0 0 -1>}
 deform_upperarm_up_r {parent=upperarm_twist_01_r, loc=<3.2214 1.1391 -3.3308>, rot=<0 0 0 -1>}
 def_neck_fur_front {parent=spine_04, loc=<14.0105 14.4709 0>, rot=<0 0 0.1217 -0.9926>}
]
