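# Physics asset for the male medium GuineaPig head cosmetic, giving its neck hair and beard bones bodies and joints for physics-driven secondary motion.
# All bodies are capsules; spine_05, neck_01 and both clavicles form the upper-body volumes.
# Nine neck-hair and three beard chains each run from a small MSTR body on neck_01 through two segments.
# MSTR, neck and clavicle joints rotate freely, while chain segment joints have tight swing and twist limits.
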PhysicsAsset /BRCosmetics/Characters/Player/Male/Medium/Heads/M_MED_GuineaPig_Head/Meshes/M_MED_GuineaPig_Head_Physics
bodies {
 clavicle_l [{shapeType=Capsule, length=11.0242, radius=7.5913, shapeName=clavicle_l_capsule, center=<8.1199 0 0>, rotation=<rot 0 -90 90>}]
 clavicle_r [{shapeType=Capsule, length=11.0242, radius=7.5913, shapeName=clavicle_l_capsule, center=<-8.1199 0 0>, rotation=<rot 0 -90 -90>}]
 deform_neck_hair_a_l_01 [{
   shapeType=Capsule
   length=2.9883
   radius=2.3349
   shapeName=deform_neck_hair_a_l_01_capsule
   center=<2.4182 0.0601 0.7406>
   rotation=<rot -31.6068 -90 22.595>
  }]
 deform_neck_hair_a_l_02 [{
   shapeType=Capsule
   length=2.5725
   radius=1.2566
   shapeName=deform_neck_hair_a_l_02_capsule
   center=<1.8063 -0.6622 0.3643>
   rotation=<rot 0 10.2294 -86.2394>
  }]
 deform_neck_hair_a_l_MSTR [{shapeType=Capsule, length=0.505, radius=0.505, shapeName=deform_neck_hair_a_l_MSTR_capsule, rotation=<rot 0 -90 0>}]
 deform_neck_hair_a_r_01 [{
   shapeType=Capsule
   length=2.9883
   radius=2.3349
   shapeName=deform_neck_hair_a_l_01_capsule
   center=<-2.4182 -0.0601 -0.7406>
   rotation=<rot -31.6068 -90 -157.405>
  }]
 deform_neck_hair_a_r_02 [{
   shapeType=Capsule
   length=2.5725
   radius=1.2566
   shapeName=deform_neck_hair_a_l_02_capsule
   center=<-1.8063 0.6622 -0.3643>
   rotation=<rot 0 10.2294 93.7606>
  }]
 deform_neck_hair_a_r_MSTR [{shapeType=Capsule, length=0.505, radius=0.505, shapeName=deform_neck_hair_a_l_MSTR_capsule, rotation=<rot 0 -90 180>}]
 deform_neck_hair_b_l_01 [{
   shapeType=Capsule
   length=3.852
   radius=2.492
   shapeName=deform_neck_hair_b_l_01_capsule
   center=<1.1966 0.6818 -0.4666>
   rotation=<rot 0 -178.432 -88.1032>
  }]
 deform_neck_hair_b_l_02 [{
   shapeType=Capsule
   length=4.5022
   radius=2.0256
   shapeName=deform_neck_hair_b_l_02_capsule
   center=<2.3781 0.3212 2.1169>
   rotation=<rot 19.2619 106.0063 -77.8218>
  }]
 deform_neck_hair_b_l_MSTR [{shapeType=Capsule, length=0.505, radius=0.505, shapeName=deform_neck_hair_b_l_MSTR_capsule, rotation=<rot 0 -90 0>}]
 deform_neck_hair_b_r_01 [{
   shapeType=Capsule
   length=3.852
   radius=2.492
   shapeName=deform_neck_hair_b_l_01_capsule
   center=<-2.4063 -0.3301 0.205>
   rotation=<rot -1.8101 -105.8353 78.26>
  }]
 deform_neck_hair_b_r_02 [{
   shapeType=Capsule
   length=3.6741
   radius=3.3538
   shapeName=deform_neck_hair_b_l_02_capsule
   center=<-3.1273 0.4135 -2.2809>
   rotation=<rot 23.7755 93.9059 100.3188>
  }]
 deform_neck_hair_b_r_MSTR [{shapeType=Capsule, length=0.505, radius=0.505, shapeName=deform_neck_hair_b_l_MSTR_capsule, rotation=<rot 0 -90 180>}]
 deform_neck_hair_c_l_01 [{
   shapeType=Capsule
   length=3.1821
   radius=2.1796
   shapeName=deform_neck_hair_c_l_01_capsule
   center=<1.4197 -0.9776 -0.1132>
   rotation=<rot 0 -178.5989 -53.7915>
  }]
 deform_neck_hair_c_l_02 [{
   shapeType=Capsule
   length=3.5508
   radius=2.0821
   shapeName=deform_neck_hair_c_l_02_capsule
   center=<1.2068 0.4192 0.9586>
   rotation=<rot 0 18.496 -89.6882>
  }]
 deform_neck_hair_c_l_MSTR [{shapeType=Capsule, length=0.505, radius=0.505, shapeName=deform_neck_hair_c_l_MSTR_capsule, rotation=<rot 0 -90 0>}]
 deform_neck_hair_c_r_01 [{
   shapeType=Capsule
   length=3.1821
   radius=2.1796
   shapeName=deform_neck_hair_c_l_01_capsule
   center=<-1.4197 0.9776 0.1132>
   rotation=<rot 0 -178.5989 126.2085>
  }]
 deform_neck_hair_c_r_02 [{
   shapeType=Capsule
   length=3.0508
   radius=2.5821
   shapeName=deform_neck_hair_c_l_02_capsule
   center=<-1.2068 -0.4192 -0.9586>
   rotation=<rot 0 18.496 90.3118>
  }]
 deform_neck_hair_c_r_MSTR [{shapeType=Capsule, length=0.505, radius=0.505, shapeName=deform_neck_hair_c_l_MSTR_capsule, rotation=<rot 0 -90 180>}]
 deform_neck_hair_d_l_01 [{
   shapeType=Capsule
   length=2.7057
   radius=2.0295
   shapeName=deform_neck_hair_d_l_01_capsule
   center=<1.8851 0.1378 -0.0302>
   rotation=<rot 0 12.4907 -80.9456>
  }]
 deform_neck_hair_d_l_02 [{
   shapeType=Capsule
   length=2.8211
   radius=1.6988
   shapeName=deform_neck_hair_d_l_02_capsule
   center=<1.4142 -0.3505 -0.1109>
   rotation=<rot 0 -53.0367 -60.9879>
  }]
 deform_neck_hair_d_l_MSTR [{shapeType=Capsule, length=0.505, radius=0.505, shapeName=deform_neck_hair_d_l_MSTR_capsule, rotation=<rot 0 -90 0>}]
 deform_neck_hair_d_r_01 [{
   shapeType=Capsule
   length=2.7057
   radius=2.0295
   shapeName=deform_neck_hair_d_l_01_capsule
   center=<-1.8851 -0.1378 0.0302>
   rotation=<rot 0 12.4907 99.0544>
  }]
 deform_neck_hair_d_r_02 [{
   shapeType=Capsule
   length=2.8211
   radius=1.6988
   shapeName=deform_neck_hair_d_l_02_capsule
   center=<-1.4142 0.3505 0.1109>
   rotation=<rot 0 -53.0367 119.0121>
  }]
 deform_neck_hair_d_r_MSTR [{shapeType=Capsule, length=0.505, radius=0.505, shapeName=deform_neck_hair_d_l_MSTR_capsule, rotation=<rot 0 -90 180>}]
 deform_neck_hair_rear_01 [{
   shapeType=Capsule
   length=2.1448
   radius=1.8165
   shapeName=deform_neck_hair_rear_01_capsule
   center=<3.403 0.2427 0.7683>
   rotation=<rot 0 -22.539 -80.5122>
  }]
 deform_neck_hair_rear_02 [{
   shapeType=Capsule
   length=3.4726
   radius=1.673
   shapeName=deform_neck_hair_rear_02_capsule
   center=<4.1122 -0.2147 -0.5209>
   rotation=<rot 0 -95.5983 -78.525>
  }]
 deform_neck_hair_rear_MSTR [{shapeType=Capsule, length=0.505, radius=0.505, shapeName=deform_neck_hair_rear_MSTR_capsule, rotation=<rot 0 -90 0>}]
 dyn_beard_c_01 [{shapeType=Capsule, length=2, radius=1.33, shapeName=dyn_beard_c_01_capsule, center=<1 0 0>, rotation=<rot 90 0 0>}]
 dyn_beard_c_02 [{shapeType=Capsule, length=2, radius=1.33, shapeName=dyn_beard_c_02_capsule, center=<1 0 0>, rotation=<rot 90 0 0>}]
 dyn_beard_c_MSTR [{shapeType=Capsule, length=0.505, radius=0.505, shapeName=dyn_beard_c_MSTR_capsule, rotation=<rot 0 -90 0>}]
 dyn_beard_l_01 [{shapeType=Capsule, length=2, radius=1.33, shapeName=dyn_beard_l_01_capsule, center=<1 0 0>, rotation=<rot 90 0 0>}]
 dyn_beard_l_02 [{shapeType=Capsule, length=2, radius=1.33, shapeName=dyn_beard_l_02_capsule, center=<1 0 0>, rotation=<rot 90 0 0>}]
 dyn_beard_l_MSTR [{shapeType=Capsule, length=0.505, radius=0.505, shapeName=dyn_beard_l_MSTR_capsule, rotation=<rot 0 -90 0>}]
 dyn_beard_r_01 [{shapeType=Capsule, length=2, radius=1.33, shapeName=dyn_beard_r_01_capsule, center=<1 0 0>, rotation=<rot 90 0 0>}]
 dyn_beard_r_02 [{shapeType=Capsule, length=2, radius=1.33, shapeName=dyn_beard_r_02_capsule, center=<1 0 0>, rotation=<rot 90 0 0>}]
 dyn_beard_r_MSTR [{shapeType=Capsule, length=0.505, radius=0.505, shapeName=dyn_beard_r_MSTR_capsule, rotation=<rot 0 -90 0>}]
 neck_01 [{
   shapeType=Capsule
   length=7.5653
   radius=9.557
   shapeName=neck_01_capsule
   center=<0.5133 3.3394 -0.005>
   rotation=<rot 0.0695 -97.1073 90.5573>
  }]
 spine_05 [{shapeType=Capsule, length=11.8682, radius=13.8741, rotation=<rot 0 -90 0>}]
}
constraints [
 {
  bone1Name=deform_neck_hair_a_l_MSTR
  bone2Name=neck_01
  swing1LimitDegrees=45
  swing1Motion=Free
  swing2LimitDegrees=45
  swing2Motion=Free
  twistLimitDegrees=45
  twistMotion=Free
 }
 {bone1Name=deform_neck_hair_a_l_01, bone2Name=deform_neck_hair_a_l_MSTR, swing1LimitDegrees=15, swing2LimitDegrees=5, twistLimitDegrees=5}
 {bone1Name=deform_neck_hair_a_l_02, bone2Name=deform_neck_hair_a_l_01, swing1LimitDegrees=15, swing2LimitDegrees=5, twistLimitDegrees=5}
 {
  bone1Name=deform_neck_hair_b_l_MSTR
  bone2Name=neck_01
  swing1LimitDegrees=45
  swing1Motion=Free
  swing2LimitDegrees=45
  swing2Motion=Free
  twistLimitDegrees=45
  twistMotion=Free
 }
 {bone1Name=deform_neck_hair_b_l_01, bone2Name=deform_neck_hair_b_l_MSTR, swing1LimitDegrees=15, swing2LimitDegrees=5, twistLimitDegrees=5}
 {bone1Name=deform_neck_hair_b_l_02, bone2Name=deform_neck_hair_b_l_01, swing1LimitDegrees=5, swing2LimitDegrees=5, twistLimitDegrees=5}
 {
  bone1Name=deform_neck_hair_c_l_MSTR
  bone2Name=neck_01
  swing1LimitDegrees=45
  swing1Motion=Free
  swing2LimitDegrees=45
  swing2Motion=Free
  twistLimitDegrees=45
  twistMotion=Free
 }
 {bone1Name=deform_neck_hair_c_l_01, bone2Name=deform_neck_hair_c_l_MSTR, swing1LimitDegrees=15, swing2LimitDegrees=5, twistLimitDegrees=5}
 {bone1Name=deform_neck_hair_c_l_02, bone2Name=deform_neck_hair_c_l_01, swing1LimitDegrees=15, swing2LimitDegrees=5, twistLimitDegrees=5}
 {
  bone1Name=deform_neck_hair_d_l_MSTR
  bone2Name=neck_01
  swing1LimitDegrees=45
  swing1Motion=Free
  swing2LimitDegrees=45
  swing2Motion=Free
  twistLimitDegrees=45
  twistMotion=Free
 }
 {bone1Name=deform_neck_hair_d_l_01, bone2Name=deform_neck_hair_d_l_MSTR, swing1LimitDegrees=15, swing2LimitDegrees=5, twistLimitDegrees=5}
 {bone1Name=deform_neck_hair_d_l_02, bone2Name=deform_neck_hair_d_l_01, swing1LimitDegrees=15, swing2LimitDegrees=5, twistLimitDegrees=5}
 {
  bone1Name=deform_neck_hair_rear_MSTR
  bone2Name=neck_01
  swing1LimitDegrees=45
  swing1Motion=Free
  swing2LimitDegrees=45
  swing2Motion=Free
  twistLimitDegrees=45
  twistMotion=Free
 }
 {bone1Name=deform_neck_hair_rear_01, bone2Name=deform_neck_hair_rear_MSTR, swing1LimitDegrees=15, swing2LimitDegrees=5, twistLimitDegrees=5}
 {bone1Name=deform_neck_hair_rear_02, bone2Name=deform_neck_hair_rear_01, swing1LimitDegrees=15, swing2LimitDegrees=5, twistLimitDegrees=5}
 {
  bone1Name=deform_neck_hair_a_r_MSTR
  bone2Name=neck_01
  swing1LimitDegrees=45
  swing1Motion=Free
  swing2LimitDegrees=45
  swing2Motion=Free
  twistLimitDegrees=45
  twistMotion=Free
 }
 {bone1Name=deform_neck_hair_a_r_01, bone2Name=deform_neck_hair_a_r_MSTR, swing1LimitDegrees=15, swing2LimitDegrees=5, twistLimitDegrees=5}
 {bone1Name=deform_neck_hair_a_r_02, bone2Name=deform_neck_hair_a_r_01, swing1LimitDegrees=15, swing2LimitDegrees=5, twistLimitDegrees=5}
 {
  bone1Name=deform_neck_hair_b_r_MSTR
  bone2Name=neck_01
  swing1LimitDegrees=45
  swing1Motion=Free
  swing2LimitDegrees=45
  swing2Motion=Free
  twistLimitDegrees=45
  twistMotion=Free
 }
 {bone1Name=deform_neck_hair_b_r_01, bone2Name=deform_neck_hair_b_r_MSTR, swing1LimitDegrees=15, swing2LimitDegrees=5, twistLimitDegrees=5}
 {bone1Name=deform_neck_hair_b_r_02, bone2Name=deform_neck_hair_b_r_01, swing1LimitDegrees=15, swing2LimitDegrees=5, twistLimitDegrees=5}
 {
  bone1Name=deform_neck_hair_c_r_MSTR
  bone2Name=neck_01
  swing1LimitDegrees=45
  swing1Motion=Free
  swing2LimitDegrees=45
  swing2Motion=Free
  twistLimitDegrees=45
  twistMotion=Free
 }
 {bone1Name=deform_neck_hair_c_r_01, bone2Name=deform_neck_hair_c_r_MSTR, swing1LimitDegrees=15, swing2LimitDegrees=5, twistLimitDegrees=5}
 {bone1Name=deform_neck_hair_c_r_02, bone2Name=deform_neck_hair_c_r_01, swing1LimitDegrees=15, swing2LimitDegrees=5, twistLimitDegrees=5}
 {
  bone1Name=deform_neck_hair_d_r_MSTR
  bone2Name=neck_01
  swing1LimitDegrees=45
  swing1Motion=Free
  swing2LimitDegrees=45
  swing2Motion=Free
  twistLimitDegrees=45
  twistMotion=Free
 }
 {bone1Name=deform_neck_hair_d_r_01, bone2Name=deform_neck_hair_d_r_MSTR, swing1LimitDegrees=15, swing2LimitDegrees=5, twistLimitDegrees=5}
 {bone1Name=deform_neck_hair_d_r_02, bone2Name=deform_neck_hair_d_r_01, swing1LimitDegrees=15, swing2LimitDegrees=5, twistLimitDegrees=5}
 {
  bone1Name=clavicle_r
  bone2Name=spine_05
  swing1LimitDegrees=45
  swing1Motion=Free
  swing2LimitDegrees=45
  swing2Motion=Free
  twistLimitDegrees=45
  twistMotion=Free
 }
 {
  bone1Name=clavicle_l
  bone2Name=spine_05
  swing1LimitDegrees=45
  swing1Motion=Free
  swing2LimitDegrees=45
  swing2Motion=Free
  twistLimitDegrees=45
  twistMotion=Free
 }
 {
  bone1Name=dyn_beard_c_MSTR
  bone2Name=neck_01
  swing1LimitDegrees=45
  swing1Motion=Free
  swing2LimitDegrees=45
  swing2Motion=Free
  twistLimitDegrees=45
  twistMotion=Free
 }
 {bone1Name=dyn_beard_c_01, bone2Name=dyn_beard_c_MSTR, swing1LimitDegrees=12, swing2LimitDegrees=12, twistLimitDegrees=12}
 {bone1Name=dyn_beard_c_02, bone2Name=dyn_beard_c_01, swing1LimitDegrees=12, swing2LimitDegrees=12, twistLimitDegrees=12}
 {
  bone1Name=dyn_beard_l_MSTR
  bone2Name=neck_01
  swing1LimitDegrees=45
  swing1Motion=Free
  swing2LimitDegrees=45
  swing2Motion=Free
  twistLimitDegrees=45
  twistMotion=Free
 }
 {bone1Name=dyn_beard_l_01, bone2Name=dyn_beard_l_MSTR, swing1LimitDegrees=12, swing2LimitDegrees=12, twistLimitDegrees=12}
 {bone1Name=dyn_beard_l_02, bone2Name=dyn_beard_l_01, swing1LimitDegrees=12, swing2LimitDegrees=12, twistLimitDegrees=12}
 {
  bone1Name=dyn_beard_r_MSTR
  bone2Name=neck_01
  swing1LimitDegrees=45
  swing1Motion=Free
  swing2LimitDegrees=45
  swing2Motion=Free
  twistLimitDegrees=45
  twistMotion=Free
 }
 {bone1Name=dyn_beard_r_01, bone2Name=dyn_beard_r_MSTR, swing1LimitDegrees=12, swing2LimitDegrees=12, twistLimitDegrees=12}
 {bone1Name=dyn_beard_r_02, bone2Name=dyn_beard_r_01, swing1LimitDegrees=12, swing2LimitDegrees=12, twistLimitDegrees=12}
 {
  bone1Name=neck_01
  bone2Name=spine_05
  swing1LimitDegrees=45
  swing1Motion=Free
  swing2LimitDegrees=45
  swing2Motion=Free
  twistLimitDegrees=45
  twistMotion=Free
 }
]
