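# Post-process AnimBP for the guinea pig head cosmetic, adding secondary motion to its beard and neck hair.
# A Control Rig deforms the input pose into a cached pose feeding two branches selected by bRigidBodyEnabled.
# True runs RigidBody plus left-hair bone tweaks; false chains AnimDynamics with planar limits, omitting left a/b/c hair.
# Each update sets Gravity Override through a deprecated state function, but no AnimGraph node visibly reads it.
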AnimBlueprint /BRCosmetics/Characters/Player/Male/Medium/Heads/M_MED_GuineaPig_Head/Meshes/M_MED_GuineaPig_Head_AnimBP_PP
target_skeleton=/Game/Characters/Player/Male/Medium/Heads/M_MED_GuineaPig_Head/Meshes/Fortnite_Base_Head_GuineaPig_Skeleton
graphs [
 AnimGraph {nodes [
   n1 = Anim Dynamics - Bone: dyn_beard_l_01(ComponentPose=Local To Component(LocalPose=Use cached pose 'PostDeformationRig'()), Alpha=1, ExternalForce="0,0,0") {
    angularDampingOverride=1
    angularSpringConstant=20000
    bAngularSpring=true
    bOverrideAngularDamping=true
    bUseGravityOverride=true
    gravityScale=0.35
    alphaBoolBlend {blendOption=Linear}
    boundBone {boneName=dyn_beard_l_01}
    componentAppliedLinearAccClamp=<200 200 200>
    componentLinearAccScale=<1 1 1>
    componentLinearVelScale=<1 1 1>
    gravityOverride {z=-980}
    physicsBodyDefinitions [{
      boundBone {boneName=dyn_beard_l_01}
      boxExtents=<20 30 30>
      constraintSetup {angularLimitsMax=<15 15 15>, angularLimitsMin=<-15 -15 -15>, angularTarget=<1 0 0>}
      localJointOffset=<1 0 0>
     }]
    relativeSpaceBone {boneName=head}
   }
   n2 = Anim Dynamics - Bone: dyn_beard_l_02(ComponentPose=@n1, Alpha=1, ExternalForce="0,0,0") {
    angularDampingOverride=1
    angularSpringConstant=20000
    bAngularSpring=true
    bOverrideAngularDamping=true
    bUseGravityOverride=true
    gravityScale=0.35
    alphaBoolBlend {blendOption=Linear}
    boundBone {boneName=dyn_beard_l_02}
    componentAppliedLinearAccClamp=<200 200 200>
    componentLinearAccScale=<1 1 1>
    componentLinearVelScale=<1 1 1>
    gravityOverride {z=-980}
    physicsBodyDefinitions [{
      boundBone {boneName=dyn_beard_l_02}
      boxExtents=<20 30 30>
      constraintSetup {angularLimitsMax=<15 15 15>, angularLimitsMin=<-15 -15 -15>, angularTarget=<1 0 0>}
      localJointOffset=<1 0 0>
     }]
    relativeSpaceBone {boneName=head}
   }
   n3 = Anim Dynamics - Bone: dyn_beard_c_01(ComponentPose=@n2, Alpha=1, ExternalForce="0,0,0") {
    angularDampingOverride=1
    angularSpringConstant=20000
    bAngularSpring=true
    bOverrideAngularDamping=true
    bUseGravityOverride=true
    gravityScale=0.35
    alphaBoolBlend {blendOption=Linear}
    boundBone {boneName=dyn_beard_c_01}
    componentAppliedLinearAccClamp=<200 200 200>
    componentLinearAccScale=<1 1 1>
    componentLinearVelScale=<1 1 1>
    gravityOverride {z=-980}
    physicsBodyDefinitions [{
      boundBone {boneName=dyn_beard_c_01}
      boxExtents=<20 30 30>
      constraintSetup {angularLimitsMax=<15 15 15>, angularLimitsMin=<-15 -15 -15>, angularTarget=<1 0 0>}
      localJointOffset=<1 0 0>
     }]
    relativeSpaceBone {boneName=head}
   }
   n4 = Anim Dynamics - Bone: dyn_beard_c_02(ComponentPose=@n3, Alpha=1, ExternalForce="0,0,0") {
    angularDampingOverride=1
    angularSpringConstant=20000
    bAngularSpring=true
    bOverrideAngularDamping=true
    bUseGravityOverride=true
    gravityScale=0.35
    alphaBoolBlend {blendOption=Linear}
    boundBone {boneName=dyn_beard_c_02}
    componentAppliedLinearAccClamp=<200 200 200>
    componentLinearAccScale=<1 1 1>
    componentLinearVelScale=<1 1 1>
    gravityOverride {z=-980}
    physicsBodyDefinitions [{
      boundBone {boneName=dyn_beard_c_02}
      boxExtents=<20 30 30>
      constraintSetup {angularLimitsMax=<15 15 15>, angularLimitsMin=<-15 -15 -15>, angularTarget=<1 0 0>}
      localJointOffset=<1 0 0>
     }]
    relativeSpaceBone {boneName=head}
   }
   n5 = Anim Dynamics - Bone: dyn_beard_r_01(ComponentPose=@n4, Alpha=1, ExternalForce="0,0,0") {
    angularDampingOverride=1
    angularSpringConstant=20000
    bAngularSpring=true
    bOverrideAngularDamping=true
    bUseGravityOverride=true
    gravityScale=0.35
    alphaBoolBlend {blendOption=Linear}
    boundBone {boneName=dyn_beard_r_01}
    componentAppliedLinearAccClamp=<200 200 200>
    componentLinearAccScale=<1 1 1>
    componentLinearVelScale=<1 1 1>
    gravityOverride {z=-980}
    physicsBodyDefinitions [{
      boundBone {boneName=dyn_beard_r_01}
      boxExtents=<20 30 30>
      constraintSetup {angularLimitsMax=<15 15 15>, angularLimitsMin=<-15 -15 -15>, angularTarget=<1 0 0>}
      localJointOffset=<1 0 0>
     }]
    relativeSpaceBone {boneName=head}
   }
   n6 = Anim Dynamics - Bone: dyn_beard_r_02(ComponentPose=@n5, Alpha=1, ExternalForce="0,0,0") {
    angularDampingOverride=1
    angularSpringConstant=20000
    bAngularSpring=true
    bOverrideAngularDamping=true
    bUseGravityOverride=true
    gravityScale=0.35
    alphaBoolBlend {blendOption=Linear}
    boundBone {boneName=dyn_beard_r_02}
    componentAppliedLinearAccClamp=<200 200 200>
    componentLinearAccScale=<1 1 1>
    componentLinearVelScale=<1 1 1>
    gravityOverride {z=-980}
    physicsBodyDefinitions [{
      boundBone {boneName=dyn_beard_r_02}
      boxExtents=<20 30 30>
      constraintSetup {angularLimitsMax=<15 15 15>, angularLimitsMin=<-15 -15 -15>, angularTarget=<1 0 0>}
      localJointOffset=<1 0 0>
     }]
    relativeSpaceBone {boneName=head}
   }
   n7 = Anim Dynamics - Bone: deform_neck_hair_d_l_01(ComponentPose=@n6, Alpha=1, ExternalForce="0,0,0") {
    angularDampingOverride=1
    angularSpringConstant=20000
    bAngularSpring=true
    bOverrideAngularDamping=true
    bUseGravityOverride=true
    gravityScale=0.35
    alphaBoolBlend {blendOption=Linear}
    boundBone {boneName=deform_neck_hair_d_l_01}
    componentAppliedLinearAccClamp=<200 200 200>
    componentLinearAccScale=<1 1 1>
    componentLinearVelScale=<1 1 1>
    gravityOverride {z=-980}
    physicsBodyDefinitions [{
      boundBone {boneName=deform_neck_hair_d_l_01}
      boxExtents=<40 50 50>
      constraintSetup {angularLimitsMax=<15 0 15>, angularLimitsMin=<-15 -5 -15>, angularTarget=<1 0 0>}
      localJointOffset=<4 0 0>
     }]
    planarLimits [{drivingBone {boneName=spine_05}, planeTransform {location=<0 -16.5 0>, rotation=<rot 0 30.0001 -90.0001>}}]
   }
   n8 = Anim Dynamics - Bone: deform_neck_hair_d_l_02(ComponentPose=@n7, Alpha=1, ExternalForce="0,0,0") {
    angularDampingOverride=1
    angularSpringConstant=20000
    bAngularSpring=true
    bOverrideAngularDamping=true
    bUseGravityOverride=true
    gravityScale=0.35
    alphaBoolBlend {blendOption=Linear}
    boundBone {boneName=deform_neck_hair_d_l_02}
    componentAppliedLinearAccClamp=<200 200 200>
    componentLinearAccScale=<1 1 1>
    componentLinearVelScale=<1 1 1>
    gravityOverride {z=-980}
    physicsBodyDefinitions [{
      boundBone {boneName=deform_neck_hair_d_l_02}
      boxExtents=<40 50 50>
      constraintSetup {angularLimitsMax=<15 0 15>, angularLimitsMin=<-15 -5 -15>, angularTarget=<1 0 0>}
      localJointOffset=<4 0 0>
     }]
    planarLimits [{drivingBone {boneName=spine_05}, planeTransform {location=<0 -17.5 0>, rotation=<rot 0 30.0001 -90.0001>}}]
   }
   n9 = Anim Dynamics - Bone: deform_neck_hair_rear_01(ComponentPose=@n8, Alpha=1, ExternalForce="0,0,0") {
    angularDampingOverride=1
    angularSpringConstant=20000
    bAngularSpring=true
    bOverrideAngularDamping=true
    bUseGravityOverride=true
    gravityScale=0.35
    alphaBoolBlend {blendOption=Linear}
    boundBone {boneName=deform_neck_hair_rear_01}
    componentAppliedLinearAccClamp=<200 200 200>
    componentLinearAccScale=<1 1 1>
    componentLinearVelScale=<1 1 1>
    gravityOverride {z=-980}
    physicsBodyDefinitions [{
      boundBone {boneName=deform_neck_hair_rear_01}
      boxExtents=<40 50 50>
      constraintSetup {angularLimitsMax=<15 0 15>, angularLimitsMin=<-15 -5 -15>, angularTarget=<1 0 0>}
      localJointOffset=<4 0 0>
     }]
    planarLimits [{drivingBone {boneName=spine_05}, planeTransform {location=<0 -15 0>, rotation=<rot 0 30 -90>}}]
   }
   n10 = Anim Dynamics - Bone: deform_neck_hair_rear_02(ComponentPose=@n9, Alpha=1, ExternalForce="0,0,0") {
    angularDampingOverride=1
    angularSpringConstant=20000
    bAngularSpring=true
    bOverrideAngularDamping=true
    bUseGravityOverride=true
    gravityScale=0.35
    alphaBoolBlend {blendOption=Linear}
    boundBone {boneName=deform_neck_hair_rear_02}
    componentAppliedLinearAccClamp=<200 200 200>
    componentLinearAccScale=<1 1 1>
    componentLinearVelScale=<1 1 1>
    gravityOverride {z=-980}
    physicsBodyDefinitions [{
      boundBone {boneName=deform_neck_hair_rear_02}
      boxExtents=<40 50 50>
      constraintSetup {angularLimitsMax=<15 0 15>, angularLimitsMin=<-15 -5 -15>, angularTarget=<1 0 0>}
      localJointOffset=<4 0 0>
     }]
    planarLimits [{drivingBone {boneName=spine_05}, planeTransform {location=<0 -15 0>, rotation=<rot 0 30.0001 -90.0001>}}]
   }
   n11 = Anim Dynamics - Bone: deform_neck_hair_d_r_01(ComponentPose=@n10, Alpha=1, ExternalForce="0,0,0") {
    angularDampingOverride=1
    angularSpringConstant=20000
    bAngularSpring=true
    bOverrideAngularDamping=true
    bUseGravityOverride=true
    gravityScale=0.35
    alphaBoolBlend {blendOption=Linear}
    boundBone {boneName=deform_neck_hair_d_r_01}
    componentAppliedLinearAccClamp=<200 200 200>
    componentLinearAccScale=<1 1 1>
    componentLinearVelScale=<1 1 1>
    gravityOverride {z=-980}
    physicsBodyDefinitions [{
      boundBone {boneName=deform_neck_hair_d_r_01}
      boxExtents=<40 50 50>
      constraintSetup {angularLimitsMax=<15 0 15>, angularLimitsMin=<-15 -5 -15>, angularTarget=<1 0 0>}
      localJointOffset=<-4 0 0>
     }]
    planarLimits [{drivingBone {boneName=spine_05}, planeTransform {location=<0 -16.5 0>, rotation=<rot 0 30.0001 -90.0001>}}]
   }
   n12 = Anim Dynamics - Bone: deform_neck_hair_d_r_02(ComponentPose=@n11, Alpha=1, ExternalForce="0,0,0") {
    angularDampingOverride=1
    angularSpringConstant=20000
    bAngularSpring=true
    bOverrideAngularDamping=true
    bUseGravityOverride=true
    gravityScale=0.35
    alphaBoolBlend {blendOption=Linear}
    boundBone {boneName=deform_neck_hair_d_r_02}
    componentAppliedLinearAccClamp=<200 200 200>
    componentLinearAccScale=<1 1 1>
    componentLinearVelScale=<1 1 1>
    gravityOverride {z=-980}
    physicsBodyDefinitions [{
      boundBone {boneName=deform_neck_hair_d_r_02}
      boxExtents=<40 50 50>
      constraintSetup {angularLimitsMax=<15 0 15>, angularLimitsMin=<-15 -5 -15>, angularTarget=<1 0 0>}
      localJointOffset=<-4 0 0>
     }]
    planarLimits [{drivingBone {boneName=spine_05}, planeTransform {location=<0 -17.5 0>, rotation=<rot 0 30.0001 -90.0001>}}]
   }
   n13 = Anim Dynamics - Bone: deform_neck_hair_a_r_01(ComponentPose=@n12, Alpha=1, ExternalForce="0,0,0") {
    angularDampingOverride=1
    angularSpringConstant=20000
    bAngularSpring=true
    bOverrideAngularDamping=true
    bUseGravityOverride=true
    gravityScale=0.35
    alphaBoolBlend {blendOption=Linear}
    boundBone {boneName=deform_neck_hair_a_r_01}
    componentAppliedLinearAccClamp=<200 200 200>
    componentLinearAccScale=<1 1 1>
    componentLinearVelScale=<1 1 1>
    gravityOverride {z=-980}
    physicsBodyDefinitions [{
      boundBone {boneName=deform_neck_hair_a_r_01}
      boxExtents=<40 50 50>
      constraintSetup {angularLimitsMax=<15 15 15>, angularLimitsMin=<-15 -15 -15>, angularTarget=<1 0 0>}
      localJointOffset=<-4 0 0>
     }]
   }
   n14 = Anim Dynamics - Bone: deform_neck_hair_a_r_02(ComponentPose=@n13, Alpha=1, ExternalForce="0,0,0") {
    angularDampingOverride=1
    angularSpringConstant=20000
    bAngularSpring=true
    bOverrideAngularDamping=true
    bUseGravityOverride=true
    gravityScale=0.35
    alphaBoolBlend {blendOption=Linear}
    boundBone {boneName=deform_neck_hair_a_r_02}
    componentAppliedLinearAccClamp=<200 200 200>
    componentLinearAccScale=<1 1 1>
    componentLinearVelScale=<1 1 1>
    gravityOverride {z=-980}
    physicsBodyDefinitions [{
      boundBone {boneName=deform_neck_hair_a_r_02}
      boxExtents=<40 50 50>
      constraintSetup {angularLimitsMax=<15 15 15>, angularLimitsMin=<-15 -15 -15>, angularTarget=<1 0 0>}
      localJointOffset=<-4 0 0>
     }]
    planarLimits [{drivingBone {boneName=clavicle_r}, planeTransform {location=<0.4406 10.4319 -8.3868>, rotation=<rot 4.4853 179.9869 -179.9772>}}]
   }
   n15 = Anim Dynamics - Bone: deform_neck_hair_b_r_01(ComponentPose=@n14, Alpha=1, ExternalForce="0,0,0") {
    angularDampingOverride=1
    angularSpringConstant=20000
    bAngularSpring=true
    bOverrideAngularDamping=true
    bUseGravityOverride=true
    gravityScale=0.35
    alphaBoolBlend {blendOption=Linear}
    boundBone {boneName=deform_neck_hair_b_r_01}
    componentAppliedLinearAccClamp=<200 200 200>
    componentLinearAccScale=<1 1 1>
    componentLinearVelScale=<1 1 1>
    gravityOverride {z=-980}
    physicsBodyDefinitions [{
      boundBone {boneName=deform_neck_hair_b_r_01}
      boxExtents=<40 50 50>
      constraintSetup {angularLimitsMax=<15 15 15>, angularLimitsMin=<-15 -15 -15>, angularTarget=<1 0 0>}
      localJointOffset=<-4 0 0>
     }]
   }
   n16 = Anim Dynamics - Bone: deform_neck_hair_b_r_02(ComponentPose=@n15, Alpha=1, ExternalForce="0,0,0") {
    angularDampingOverride=1
    angularSpringConstant=20000
    bAngularSpring=true
    bOverrideAngularDamping=true
    bUseGravityOverride=true
    gravityScale=0.35
    alphaBoolBlend {blendOption=Linear}
    boundBone {boneName=deform_neck_hair_b_r_02}
    componentAppliedLinearAccClamp=<200 200 200>
    componentLinearAccScale=<1 1 1>
    componentLinearVelScale=<1 1 1>
    gravityOverride {z=-980}
    physicsBodyDefinitions [{
      boundBone {boneName=deform_neck_hair_b_r_02}
      boxExtents=<40 50 50>
      constraintSetup {angularLimitsMax=<15 15 15>, angularLimitsMin=<-15 -15 -15>, angularTarget=<1 0 0>}
      localJointOffset=<-4 0 0>
     }]
    planarLimits [{drivingBone {boneName=clavicle_r}, planeTransform {location=<0.4406 10.4319 -8.3868>, rotation=<rot 4.4853 179.9869 -179.9772>}}]
   }
   n17 = Anim Dynamics - Bone: deform_neck_hair_c_r_01(ComponentPose=@n16, Alpha=1, ExternalForce="0,0,0") {
    angularDampingOverride=1
    angularSpringConstant=20000
    bAngularSpring=true
    bOverrideAngularDamping=true
    bUseGravityOverride=true
    gravityScale=0.35
    alphaBoolBlend {blendOption=Linear}
    boundBone {boneName=deform_neck_hair_c_r_01}
    componentAppliedLinearAccClamp=<200 200 200>
    componentLinearAccScale=<1 1 1>
    componentLinearVelScale=<1 1 1>
    gravityOverride {z=-980}
    physicsBodyDefinitions [{
      boundBone {boneName=deform_neck_hair_c_r_01}
      boxExtents=<40 50 50>
      constraintSetup {angularLimitsMax=<15 15 15>, angularLimitsMin=<-15 -15 -15>, angularTarget=<1 0 0>}
      localJointOffset=<-4 0 0>
     }]
   }
   n18 = Anim Dynamics - Bone: deform_neck_hair_c_r_02(ComponentPose=@n17, Alpha=1, ExternalForce="0,0,0") {
    angularDampingOverride=1
    angularSpringConstant=20000
    bAngularSpring=true
    bOverrideAngularDamping=true
    bUseGravityOverride=true
    gravityScale=0.35
    alphaBoolBlend {blendOption=Linear}
    boundBone {boneName=deform_neck_hair_c_r_02}
    componentAppliedLinearAccClamp=<200 200 200>
    componentLinearAccScale=<1 1 1>
    componentLinearVelScale=<1 1 1>
    gravityOverride {z=-980}
    physicsBodyDefinitions [{
      boundBone {boneName=deform_neck_hair_c_r_02}
      boxExtents=<40 50 50>
      constraintSetup {angularLimitsMax=<15 15 15>, angularLimitsMin=<-15 -15 -15>, angularTarget=<1 0 0>}
      localJointOffset=<-4 0 0>
     }]
    planarLimits [{drivingBone {boneName=clavicle_r}, planeTransform {location=<0.4406 10.4319 -8.3868>, rotation=<rot 19.2355 179.9874 -179.9795>}}]
   }
   n19 = RigidBody(ComponentPose=Local To Component(LocalPose=Use cached pose 'PostDeformationRig'())) {
    overridePhysicsAsset=&/M_MED_GuineaPig_Head_Physics
    simulationSpace=BaseBoneSpace
    alphaBoolBlend {blendOption=Linear}
    baseBoneRef {boneName=spine_04}
    componentAppliedLinearAccClamp=<1300 1300 1300>
   }
   n20 = "Transform (Modify) Bone - Bone: deform_neck_hair_b_l_MSTR"(ComponentPose=@n19, Alpha=1, Translation="0,0,0", Rotation="0,0,0", Scale="0.8,1,1") {
    scaleMode=BMM_Replace
    scaleSpace=BCS_BoneSpace
    alphaBoolBlend {blendOption=Linear}
    boneToModify {boneName=deform_neck_hair_b_l_MSTR}
    scale {x=0.8}
   }
   n21 = "Transform (Modify) Bone - Bone: deform_neck_hair_b_l_02"(ComponentPose=@n20, Alpha=1, Translation="0,0,0", Rotation="-15.391,-9.279,2.711", Scale="0.6,1,1") {
    rotationMode=BMM_Additive
    rotationSpace=BCS_BoneSpace
    scaleMode=BMM_Replace
    scaleSpace=BCS_BoneSpace
    alphaBoolBlend {blendOption=Linear}
    boneToModify {boneName=deform_neck_hair_b_l_02}
    rotation=<rot -15.391 -9.279 2.711>
    scale {x=0.6}
   }
   Output Pose(Result=Blend Poses by bool(BlendPose_0=Component To Local(ComponentPose=@n21), BlendPose_1=Component To Local(ComponentPose=@n18), bActiveValue=Get bRigidBodyEnabled()) {blendType=Linear})
   n22 = Control Rig(Source=Input Pose()) {
    bResetInputPoseToInitial=false
    bTransferInputCurves=false
    alphaBoolBlend {blendOption=Linear}
    controlRigAssetReference {blueprintRigClass=&/M_MED_GuineaPig_Head_CtrlRig.M_MED_GuineaPig_Head_CtrlRig_C}
    outputBonesToTransfer [
     {boneName=deform_neck_hair_a_l_MSTR}
     {boneName=deform_neck_hair_a_l_01}
     {boneName=deform_neck_hair_a_l_02}
     {boneName=deform_neck_hair_b_l_MSTR}
     {boneName=deform_neck_hair_b_l_01}
     {boneName=deform_neck_hair_b_l_02}
     {boneName=deform_neck_hair_c_l_MSTR}
     {boneName=deform_neck_hair_c_l_01}
     {boneName=deform_neck_hair_c_l_02}
     {boneName=deform_neck_hair_a_r_MSTR}
     {boneName=deform_neck_hair_a_r_01}
     {boneName=deform_neck_hair_a_r_02}
     {boneName=deform_neck_hair_b_r_MSTR}
     {boneName=deform_neck_hair_b_r_01}
     {boneName=deform_neck_hair_b_r_02}
     {boneName=deform_neck_hair_c_r_MSTR}
     {boneName=deform_neck_hair_c_r_01}
     {boneName=deform_neck_hair_c_r_02}
     {boneName=deform_neck_hair_d_l_MSTR}
     {boneName=deform_neck_hair_d_l_01}
     {boneName=deform_neck_hair_d_l_02}
     {boneName=deform_neck_hair_rear_MSTR}
     {boneName=deform_neck_hair_rear_01}
     {boneName=deform_neck_hair_rear_02}
     {boneName=deform_neck_hair_d_r_MSTR}
     {boneName=deform_neck_hair_d_r_01}
     {boneName=deform_neck_hair_d_r_02}
     {boneName=dyn_beard_MSTR}
    ]
   }
   Save cached pose 'PostDeformationRig'(Pose=@n22) {cacheName=PostDeformationRig}
  ]}
 EventGraph {nodes [
   Event BlueprintUpdateAnimation()
   > n1 = Gravity Override State Function DEPRECATED(CharacterPartAnimInstance=Self-Reference(), GameTimeInSeconds=GetGameTimeInSeconds(), GravityOverrideParams=Get Gravity Override Params())
   > Set Gravity Override(Gravity Override=@n1)
  ]}
]
variables [
 ClothEnabled {type=number}
 Gravity Override {type=Vector, default=<0 0 -3000>}
 Gravity Override Params {
  type=GravityOverrideParamsStruct
  default {
   emoteOrMelee {gravityOverride=<0 0 -3000>, windAmplitude=<0 200 150>, windFrequency=<3 10 8>, windOffset=<0 20 0>}
   front_End {gravityOverride=<0 0 -3000>, windAmplitude=<0 200 150>, windFrequency=<3 10 8>, windOffset=<0 20 0>}
   onGround_Falling {gravityOverride=<0 0 -3000>, windAmplitude=<0 200 150>, windFrequency=<3 10 8>, windOffset=<0 20 0>}
   onGround_Moving {gravityOverride=<0 0 -3000>, windAmplitude=<0 200 150>, windFrequency=<3 10 8>, windOffset=<0 20 0>}
   onGround_Standing {gravityOverride=<0 0 -4000>, windAmplitude=<0 150 100>, windFrequency=<1.5 8 6>, windOffset=<0 20 0>}
   rail_Grinding {gravityOverride=<0 0 -3000>, windAmplitude=<0 200 150>, windFrequency=<3 10 8>, windOffset=<0 20 0>}
   riding_Moving {gravityOverride=<0 0 -3000>, windAmplitude=<0 200 150>, windFrequency=<3 10 8>, windOffset=<0 20 0>}
   riding_Moving_Targeting {gravityOverride=<0 0 -3000>, windAmplitude=<0 200 150>, windFrequency=<3 10 8>, windOffset=<0 20 0>}
   riding_Standing {gravityOverride=<0 0 -3000>, windAmplitude=<0 200 150>, windFrequency=<3 10 8>, windOffset=<0 20 0>}
   skydive_DivingDown {jointName=root, gravityOverride=<0 700 -1000>, windAmplitude=<50 -5000 100>, windFrequency=<5 30 5>, windOffset=<-200 -500 500>}
   skydive_FloatingDown {gravityOverride=<0 700 -1000>, windAmplitude=<50 -2500 100>, windFrequency=<5 15 5>, windOffset=<-200 -500 500>}
   skydive_GoingUp {gravityOverride=<0 700 -1000>, windAmplitude=<50 -2500 100>, windFrequency=<5 15 5>, windOffset=<-200 -500 500>}
   skydive_Parachuting {gravityOverride=<0 700 -1000>, windAmplitude=<50 -2500 100>, windFrequency=<5 15 5>, windOffset=<-200 -500 500>}
   skydive_Surfing {gravityOverride=<0 0 -3000>, windAmplitude=<0 200 150>, windFrequency=<3 10 8>, windOffset=<0 20 0>}
   swimming {jointName=root, gravityOverride=<0 0 -3000>, windAmplitude=<0 200 150>, windFrequency=<3 10 8>, windOffset=<0 20 0>}
   tactical_Sprint {gravityOverride=<0 0 -3000>, windAmplitude=<0 200 150>, windFrequency=<3 10 8>, windOffset=<0 20 0>}
   ziplining {gravityOverride=<0 0 -3000>, windAmplitude=<0 200 150>, windFrequency=<3 10 8>, windOffset=<0 20 0>}
  }
 }
 Output {type=SkeletalMeshComponentBudgeted}
]
